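# Sound cue for the Marionette emote's join foley, layering a fireworks burst with a crowd reaction.
# A mixer sums two modulated branches, each picking equally at random: one from eight fireworks waves, the other from five crowd waves.
# It uses emote foley class, attenuation and concurrency settings, keeps playing when silent, and sends to a music audio-analysis submix.
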SoundCue /BRCosmetics/Sounds/Emotes/Marionette/EmoteFoley_Marionette_Join_Cue
root=Mixer_0
nodes [
 {id=Mixer_0, children [Modulator_1,Modulator_0], properties {inputVolume [1,1]}}
 {id=Modulator_0, children [Random_1]}
 {id=Modulator_1, children [Random_0]}
 {
  id=Random_0
  children [WavePlayer_1,WavePlayer_2,WavePlayer_3,WavePlayer_4,WavePlayer_5,WavePlayer_6,WavePlayer_7,WavePlayer_0]
  properties {weights [1,1,1,1,1,1,1,1]}
 }
 {id=Random_1, children [WavePlayer_13, WavePlayer_14, WavePlayer_15, WavePlayer_16, WavePlayer_17], properties {weights [1,1,1,1,1]}}
 {id=WavePlayer_0, properties {soundWaveAssetPtr=&/Emote_Marionette_Fireworks_08}}
 {id=WavePlayer_1, properties {soundWaveAssetPtr=&/Emote_Marionette_Fireworks_01}}
 {id=WavePlayer_13, properties {soundWaveAssetPtr=&/Emote_Marionette_Crowd_05}}
 {id=WavePlayer_14, properties {soundWaveAssetPtr=&/Emote_Marionette_Crowd_01}}
 {id=WavePlayer_15, properties {soundWaveAssetPtr=&/Emote_Marionette_Crowd_02}}
 {id=WavePlayer_16, properties {soundWaveAssetPtr=&/Emote_Marionette_Crowd_03}}
 {id=WavePlayer_17, properties {soundWaveAssetPtr=&/Emote_Marionette_Crowd_04}}
 {id=WavePlayer_2, properties {soundWaveAssetPtr=&/Emote_Marionette_Fireworks_02}}
 {id=WavePlayer_3, properties {soundWaveAssetPtr=&/Emote_Marionette_Fireworks_03}}
 {id=WavePlayer_4, properties {soundWaveAssetPtr=&/Emote_Marionette_Fireworks_04}}
 {id=WavePlayer_5, properties {soundWaveAssetPtr=&/Emote_Marionette_Fireworks_05}}
 {id=WavePlayer_6, properties {soundWaveAssetPtr=&/Emote_Marionette_Fireworks_06}}
 {id=WavePlayer_7, properties {soundWaveAssetPtr=&/Emote_Marionette_Fireworks_07}}
]
properties {
 attenuationSettings=&Fort_Emotes_Athena_Att
 audioPropertiesSheet=/BRCosmetics/Sounds/Settings/PropertySheets/Emotes/APS_Cosmetic_Emote_Foley
 maxDistance=1600
 soundClassObject=/Game/Sounds/FortSoundClassesAndModes/Fort_Foley_Emote
 virtualizationMode=PlayWhenSilent
 volumeMultiplier=0.8
 concurrencyOverrides {maxCount=3}
 concurrencySet [/Game/Sounds/Concurrency_Rules/Athena_EmoteFoley_Conc]
 soundSubmixSends [{minSendLevel=0, soundSubmix=&Fortnite_AudioAnalysis_Music}]
}
